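# M_Overlay_VertexCrunch is an unlit material for skeletal, morph-target and mesh-deformer meshes, named as an overlay that crunches vertices.
# Its only expression subtracts world position from object position, giving the vector from each point to the object's origin.
# Used as a world position offset, that vector would collapse vertices onto the origin, though no output connection is shown.
# Ray-traced shadow casting is disabled.
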Material /Game/Global/Materials/M_Overlay_VertexCrunch
graph {
 properties {
  bCastRayTracedShadows=false
  bUsedWithMeshDeformer=true
  bUsedWithMorphTargets=true
  bUsedWithSkeletalMesh=true
  pixelDepthOffsetMode=PDOM_Legacy
  shadingModel=MSM_Unlit
  usedShadingModels=Unlit
 }
 nodes [Subtract(A=ObjectPositionWS(), B=WorldPosition().XYZ)]
}
